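# Spot ambient Sound Cue that plays intermittent wood stress creaks for a wooden corridor bridge point of interest.
# A looping node repeatedly triggers a random delay of several seconds, then a Random node picks one of ten equally weighted Primitive wood-creak one-shots.
# The cue is slightly pitched down and reduced in volume, using one-shot tile ambient attenuation and the outdoor ambient sound class.
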SoundCue /Game/Sounds/Artemis/POI/Normal_Bridges/Bridges_Wooden_Corridor/Spot_Bridge_Wooden_Corridor_Stress_Creaks_Cue
root=Looping_0
nodes [
 {id=Delay_0, children [Random_2], properties {delayMax=15, delayMin=5}}
 {id=Looping_0, children [Delay_0]}
 {
  id=Random_2
  children [
   WavePlayer_32
   WavePlayer_33
   WavePlayer_34
   WavePlayer_35
   WavePlayer_36
   WavePlayer_37
   WavePlayer_38
   WavePlayer_39
   WavePlayer_40
   WavePlayer_41
  ]
  properties {weights [1,1,1,1,1,1,1,1,1,1]}
 }
 {id=WavePlayer_32, properties {soundWaveAssetPtr=/Game/Sounds/Fort_Ambient_NonLoop/Primitive/Primitive_Structure_WoodCreaks_01}}
 {id=WavePlayer_33, properties {soundWaveAssetPtr=/Game/Sounds/Fort_Ambient_NonLoop/Primitive/Primitive_Structure_WoodCreaks_04}}
 {id=WavePlayer_34, properties {soundWaveAssetPtr=/Game/Sounds/Fort_Ambient_NonLoop/Primitive/Primitive_Structure_WoodCreaks_05}}
 {id=WavePlayer_35, properties {soundWaveAssetPtr=/Game/Sounds/Fort_Ambient_NonLoop/Primitive/Primitive_Structure_WoodCreaks_10}}
 {id=WavePlayer_36, properties {soundWaveAssetPtr=/Game/Sounds/Fort_Ambient_NonLoop/Primitive/Primitive_Structure_WoodCreaks_02}}
 {id=WavePlayer_37, properties {soundWaveAssetPtr=/Game/Sounds/Fort_Ambient_NonLoop/Primitive/Primitive_Structure_WoodCreaks_03}}
 {id=WavePlayer_38, properties {soundWaveAssetPtr=/Game/Sounds/Fort_Ambient_NonLoop/Primitive/Primitive_Structure_WoodCreaks_06}}
 {id=WavePlayer_39, properties {soundWaveAssetPtr=/Game/Sounds/Fort_Ambient_NonLoop/Primitive/Primitive_Structure_WoodCreaks_07}}
 {id=WavePlayer_40, properties {soundWaveAssetPtr=/Game/Sounds/Fort_Ambient_NonLoop/Primitive/Primitive_Structure_WoodCreaks_08}}
 {id=WavePlayer_41, properties {soundWaveAssetPtr=/Game/Sounds/Fort_Ambient_NonLoop/Primitive/Primitive_Structure_WoodCreaks_09}}
]
properties {
 attenuationSettings=/Game/Sounds/Attenuation/Fort_TileAmbient_OneShot_Att
 maxDistance=8000
 pitchMultiplier=0.9
 soundClassObject=&Fort_Ambient_Outdoors_Athena
 volumeMultiplier=0.401
 preEffectBusSends [{}]
}
